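# Physics asset for the gem mesh of the HydroIgnite backpack cosmetic, giving capsule bodies and joints to its dangling metal chips.
# The pack body connects to a base chip body through a joint with free swing and twist.
# Four two-link chains, one front and three back, branch from that base, each a short capsule followed by a longer one.
# Chain joints carry individual swing and twist limits, generally wider at the outer links.
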PhysicsAsset /BRCosmetics/Accessories/FORT_Backpacks/Backpack_M_MED_HydroIgnite/Meshes/M_MED_HydroIgnite_Pack_Gem_Physics
bodies {
 base_metal_chips [{shapeType=Capsule, length=0.505, radius=0.505, shapeName=base_metal_chips_capsule, rotation=<rot 0 -90 0>}]
 dyn_back_metal_chip1 [{shapeType=Capsule, length=6.5, radius=1, shapeName=dyn_front_metal_chip_capsule, center=<0.5 0 0>}]
 dyn_back_metal_chip2 [{shapeType=Capsule, length=4.5, radius=1, shapeName=dyn_front_metal_chip_capsule, center=<0.5 0 0>}]
 dyn_back_metal_chip3 [{shapeType=Capsule, length=6.5, radius=0.75, shapeName=dyn_front_metal_chip_capsule, center=<0.5 0 0>}]
 dyn_back_metal_chip_01_1 [{shapeType=Capsule, length=1, radius=1, shapeName=dyn_back_metal_chip_1_capsule, center=<1 0 0>, rotation=<rot 90 0 0>}]
 dyn_back_metal_chip_02_1 [{shapeType=Capsule, length=1, radius=1, shapeName=dyn_back_metal_chip_1_capsule, center=<1 0 0>, rotation=<rot -90 0 0>}]
 dyn_back_metal_chip_03_1 [{shapeType=Capsule, length=1, radius=1, shapeName=dyn_back_metal_chip_1_capsule, center=<1 0 0>, rotation=<rot -90 0 0>}]
 dyn_front_metal_chip1 [{shapeType=Capsule, length=9, radius=0.75, shapeName=dyn_front_metal_chip_capsule, center=<0.5 0 0>}]
 dyn_front_metal_chip_01 [{shapeType=Capsule, length=1, radius=1, shapeName=dyn_front_metal_chip_capsule, center=<1 0 0>, rotation=<rot 90 0 0>}]
 pack [{shapeType=Capsule, length=2, radius=1, shapeName=pack_capsule, center=<2 0 0>, rotation=<rot 90 0 0>}]
}
constraints [
 {
  bone1Name=base_metal_chips
  bone2Name=pack
  swing1LimitDegrees=45
  swing1Motion=Free
  swing2LimitDegrees=45
  swing2Motion=Free
  twistLimitDegrees=45
  twistMotion=Free
 }
 {bone1Name=dyn_back_metal_chip_01_1, bone2Name=base_metal_chips, swing1LimitDegrees=60, swing2LimitDegrees=15, twistLimitDegrees=2}
 {bone1Name=dyn_back_metal_chip_02_1, bone2Name=base_metal_chips, swing1LimitDegrees=90, swing2LimitDegrees=25, twistLimitDegrees=15}
 {bone1Name=dyn_back_metal_chip_03_1, bone2Name=base_metal_chips, swing1LimitDegrees=45, swing2LimitDegrees=10, twistLimitDegrees=5}
 {bone1Name=dyn_front_metal_chip1, bone2Name=dyn_front_metal_chip_01, swing1LimitDegrees=120, swing2LimitDegrees=60, twistLimitDegrees=72}
 {bone1Name=dyn_front_metal_chip_01, bone2Name=base_metal_chips, swing1LimitDegrees=120, swing2LimitDegrees=45, twistLimitDegrees=10}
 {bone1Name=dyn_back_metal_chip3, bone2Name=dyn_back_metal_chip_03_1, swing1LimitDegrees=120, swing2LimitDegrees=80, twistLimitDegrees=72}
 {bone1Name=dyn_back_metal_chip2, bone2Name=dyn_back_metal_chip_02_1, swing1LimitDegrees=90, swing2LimitDegrees=45, twistLimitDegrees=72}
 {bone1Name=dyn_back_metal_chip1, bone2Name=dyn_back_metal_chip_01_1, swing1LimitDegrees=90, swing2LimitDegrees=45, twistLimitDegrees=72}
]
