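# Physics asset for the HydroIgnite backpack cosmetic, defining secondary motion for its front handle and side flap.
# Small capsule bodies cover the pack bone, two base bones and two dynamic chains: four handle segments and three flap segments.
# Constraints link each chain through its base to the pack.
# Handle segments swing mainly on one axis with swing2 locked, while flap segments get looser limited cone and twist.
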PhysicsAsset /BRCosmetics/Accessories/FORT_Backpacks/Backpack_M_MED_HydroIgnite/Meshes/M_MED_HydroIgnite_Pack_Physics
bodies {
 base_front_handle [{shapeType=Capsule, length=2, radius=1, shapeName=base_front_handle_capsule, center=<2 0 0>, rotation=<rot 90 0 0>}]
 base_side_flap [{shapeType=Capsule, length=2, radius=1, shapeName=base_side_flap_capsule, center=<2 0 0>, rotation=<rot 90 0 0>}]
 dyn_front_a_handle_1 [{shapeType=Capsule, length=0.25, radius=1, shapeName=dyn_front_a_handle_1_capsule, center=<0.5 0 0>, rotation=<rot 90 0 0>}]
 dyn_front_a_handle_2 [{shapeType=Capsule, length=0.75, radius=1, shapeName=dyn_front_a_handle_2_capsule, center=<1.25 0 0>, rotation=<rot 90 0 0>}]
 dyn_front_a_handle_3 [{shapeType=Capsule, length=0.75, radius=1, shapeName=dyn_front_a_handle_3_capsule, center=<1 0 0>, rotation=<rot 90 0 0>}]
 dyn_front_a_handle_4 [{shapeType=Capsule, length=0.75, radius=1, shapeName=dyn_front_a_handle_4_capsule, center=<1 0 0>, rotation=<rot 90 0 0>}]
 dyn_side_flap_1 [{shapeType=Capsule, length=1.5, radius=1, shapeName=dyn_side_flap_1_capsule, center=<1 0 0>, rotation=<rot 90 0 0>}]
 dyn_side_flap_2 [{shapeType=Capsule, length=1.5, radius=1, shapeName=dyn_side_flap_2_capsule, center=<1 0 0>, rotation=<rot 90 0 0>}]
 dyn_side_flap_3 [{shapeType=Capsule, length=1.5, radius=1, shapeName=dyn_side_flap_3_capsule, center=<1 0 0>, rotation=<rot 90 0 0>}]
 pack [{shapeType=Capsule, length=2, radius=1, shapeName=pack_capsule, center=<2 0 0>, rotation=<rot 90 0 0>}]
}
constraints [
 {
  bone1Name=base_front_handle
  bone2Name=pack
  swing1LimitDegrees=45
  swing1Motion=Free
  swing2LimitDegrees=45
  swing2Motion=Locked
  twistLimitDegrees=45
  twistMotion=Locked
 }
 {
  bone1Name=dyn_front_a_handle_1
  bone2Name=base_front_handle
  swing1LimitDegrees=22
  swing2LimitDegrees=5
  swing2Motion=Locked
  twistLimitDegrees=2
 }
 {
  bone1Name=dyn_front_a_handle_2
  bone2Name=dyn_front_a_handle_1
  swing1LimitDegrees=25
  swing2LimitDegrees=5
  swing2Motion=Locked
  twistLimitDegrees=2
 }
 {
  bone1Name=dyn_front_a_handle_3
  bone2Name=dyn_front_a_handle_2
  swing1LimitDegrees=25
  swing2LimitDegrees=5
  swing2Motion=Locked
  twistLimitDegrees=2
 }
 {
  bone1Name=dyn_front_a_handle_4
  bone2Name=dyn_front_a_handle_3
  swing1LimitDegrees=25
  swing2LimitDegrees=5
  swing2Motion=Locked
  twistLimitDegrees=2
 }
 {
  bone1Name=base_side_flap
  bone2Name=pack
  swing1LimitDegrees=45
  swing1Motion=Free
  swing2LimitDegrees=45
  swing2Motion=Free
  twistLimitDegrees=45
  twistMotion=Free
 }
 {bone1Name=dyn_side_flap_1, bone2Name=base_side_flap, swing1LimitDegrees=20, swing2LimitDegrees=15, twistLimitDegrees=15}
 {bone1Name=dyn_side_flap_2, bone2Name=dyn_side_flap_1, swing1LimitDegrees=20, swing2LimitDegrees=15, twistLimitDegrees=15}
 {bone1Name=dyn_side_flap_3, bone2Name=dyn_side_flap_2, swing1LimitDegrees=20, swing2LimitDegrees=15, twistLimitDegrees=15}
]
