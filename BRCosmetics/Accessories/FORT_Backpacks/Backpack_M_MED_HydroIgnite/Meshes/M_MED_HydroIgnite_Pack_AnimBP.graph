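# Animation Blueprint for the HydroIgnite backpack cosmetic, animating the pack and its floating metal-chip gems.
# The pack shifts with owner spine bend plus a fixed offset; the computed backpack_rotate is unused.
# Two chained RigidBody passes follow: gems (disabled while parachuting), then a pack physics asset with dynamic gravity override.
# Four additive bone transforms then bob the front and back chips using time-driven sine waves from the update event.
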AnimBlueprint /BRCosmetics/Accessories/FORT_Backpacks/Backpack_M_MED_HydroIgnite/Meshes/M_MED_HydroIgnite_Pack_AnimBP
target_skeleton=&/M_MED_HydroIgnite_Pack_Skeleton
graphs [
 AnimGraph {nodes [
   n1 = "Transform (Modify) Bone - Bone: dyn_front_metal_chip"(
    ComponentPose=Local To Component(LocalPose=Use cached pose 'Rigid Body cache'())
    Alpha=1
    Translation="Get Translate Front Gem "() -> ("Translate Front Gem ":Vector)
    Rotation="0,0,0"
    Scale="1,1,1"
   ) {
    translationMode=BMM_Additive
    translationSpace=BCS_BoneSpace
    alphaBoolBlend {blendOption=Linear}
    boneToModify {boneName=dyn_front_metal_chip}
   }
   n2 = "Transform (Modify) Bone - Bone: dyn_back_metal_chip_1"(ComponentPose=@n1, Alpha=1, Translation=Get Translate Back Gem 1() -> (Translate Back Gem 1:Vector), Rotation="0,0,0", Scale="1,1,1") {
    translationMode=BMM_Additive
    translationSpace=BCS_BoneSpace
    alphaBoolBlend {blendOption=Linear}
    boneToModify {boneName=dyn_back_metal_chip_1}
   }
   n3 = "Transform (Modify) Bone - Bone: dyn_back_metal_chip_2"(ComponentPose=@n2, Alpha=1, Translation=Get Translate Back Gem 2() -> (Translate Back Gem 2:Vector), Rotation="0,0,0", Scale="1,1,1") {
    translationMode=BMM_Additive
    translationSpace=BCS_BoneSpace
    alphaBoolBlend {blendOption=Linear}
    boneToModify {boneName=dyn_back_metal_chip_2}
   }
   n4 = "Transform (Modify) Bone - Bone: dyn_back_metal_chip_3"(ComponentPose=@n3, Alpha=1, Translation=Get Translate Back Gem 3() -> (Translate Back Gem 3:Vector), Rotation="0,0,0", Scale="1,1,1") {
    translationMode=BMM_Additive
    translationSpace=BCS_BoneSpace
    alphaBoolBlend {blendOption=Linear}
    boneToModify {boneName=dyn_back_metal_chip_3}
   }
   Output Pose(Result=Component To Local(ComponentPose=@n4)) {layerGroup=None}
   n5 = "Transform (Modify) Bone - Bone: pack"(ComponentPose=Local To Component(LocalPose=Input Pose()), Translation=Get backpack_translation(), Rotation="0,0,0") {
    rotationMode=BMM_Additive
    rotationSpace=BCS_BoneSpace
    translationMode=BMM_Additive
    translationSpace=BCS_BoneSpace
    alphaBoolBlend {blendOption=Linear}
    boneToModify {boneName=pack}
   }
   n6 = "Transform (Modify) Bone - Bone: pack"(ComponentPose=@n5, Translation="0,-1.75,0", Rotation="0,0,0") {
    rotationMode=BMM_Additive
    rotationSpace=BCS_BoneSpace
    translationMode=BMM_Additive
    translationSpace=BCS_BoneSpace
    alphaBoolBlend {blendOption=Linear}
    boneToModify {boneName=pack}
    translation {y=-1.75}
   }
   n7 = RigidBody!(ComponentPose=@n6, Alpha=Get Gems RB Alpha() -> (Gems RB Alpha:double), OverrideWorldGravity:Vector="0,0,0", ExternalForce) -> (Pose) {
    bClampLinearTranslationLimitToRefPose=true
    bOverrideWorldGravity=true
    overridePhysicsAsset=&/M_MED_HydroIgnite_Pack_Gem_Physics
    simulationSpace=BaseBoneSpace
    alphaBoolBlend {blendOption=Linear}
    alphaScaleBiasClamp {bInterpResult=true}
    baseBoneRef {boneName=base_metal_chips}
    componentAppliedLinearAccClamp=<3000 3000 3000>
    componentLinearAccScale=<0.4 0.3 0.3>
    componentLinearVelScale=<0.4 0.3 0.3>
    overrideWorldGravity {z=-980}
    simSpaceSettings {velocityScaleZ=0.2, worldAlpha=0.1, externalLinearDragV=<0.2 0.2>}
   }
   n8 = RigidBody(ComponentPose=@n7, OverrideWorldGravity=Get Gravity Override()) {
    bClampLinearTranslationLimitToRefPose=true
    bOverrideWorldGravity=true
    overridePhysicsAsset=&/M_MED_HydroIgnite_Pack_Physics
    simulationSpace=BaseBoneSpace
    alphaBoolBlend {blendOption=Linear}
    baseBoneRef {boneName=root}
    componentAppliedLinearAccClamp=<3000 3000 3000>
    componentLinearAccScale=<0.4 0.3 0.3>
    componentLinearVelScale=<0.4 0.3 0.3>
    overrideWorldGravity {z=-980}
    simSpaceSettings {velocityScaleZ=0.5, worldAlpha=0.2, externalLinearDragV=<0.2 0.2>}
   }
   Save cached pose 'Rigid Body cache'(Pose=Component To Local(ComponentPose=@n8)) {cacheName=Rigid Body cache}
  ]}
 EventGraph {nodes [
   Event BlueprintUpdateAnimation()
   > n1 = Sequence!(execute) -> (then_0, then_1, then_2:Exec, then_3:Exec, then_4:Exec, then_5:Exec)
   > n2 = Gravity Override State Function DEPRECATED(CharacterPartAnimInstance=Self-Reference(), GameTimeInSeconds=GetGameTimeInSeconds(), GravityOverrideParams=Get Gravity Override Params())
   > Set Gravity Override(Gravity Override=@n2)
   > Set Gems RB Alpha(Gems RB Alpha:double=SelectFloat(A=0, B=1, bPickA=Get bIsParachuteOpened())) -> (Output_Get:double)
   n3 = GetRealTimeSeconds()
   n4 = "Sin (Radians)"(A=float * float(A=float * float(A=-1, B=-0.9) {is=title K2Node_PromotableOperator}, B=@n3))
   n5 = "Sin (Radians)"(A=float * float(A=float * float(A=-1.2, B=-0.8) {is=title K2Node_PromotableOperator}, B=@n3) {is=title K2Node_CommutativeAssociativeBinaryOperator})
   n6 = "Sin (Radians)"(A=float * float(A=float * float(A=0.5, B=-1.1) {is=title K2Node_PromotableOperator}, B=@n3) {is=title K2Node_CommutativeAssociativeBinaryOperator})
   Set Translate Back Gem 3(
    execute=@n1.then_4
    Translate Back Gem 3:Vector=MakeVector(X=float / float(A=@n4, B=-0.45), Y=float / float(A=@n6, B=0.8), Z=float / float(A=@n5, B=-0.7))
   ) -> (Output_Get:Vector)
   n7 = "Sin (Radians)"(A=float * float(A=float * float(A=-1.3, B=-0.8) {is=title K2Node_PromotableOperator}, B=@n3) {is=title K2Node_CommutativeAssociativeBinaryOperator})
   n8 = "Sin (Radians)"(A=float * float(A=float * float(A=0.4, B=-0.8) {is=title K2Node_PromotableOperator}, B=@n3) {is=title K2Node_CommutativeAssociativeBinaryOperator})
   n9 = "Sin (Radians)"(A=float * float(A=float * float(A=1.25, B=1.5) {is=title K2Node_PromotableOperator}, B=@n3) {is=title K2Node_CommutativeAssociativeBinaryOperator})
   Set Translate Back Gem 1(
    execute=@n1.then_2
    Translate Back Gem 1:Vector=MakeVector(X=float / float(A=@n7, B=-1), Y=float / float(A=@n9, B=-0.8), Z=float / float(A=@n8, B=-0.75))
   ) -> (Output_Get:Vector)
   n10 = GetOwnerSkeletalMeshComponent()
   n11 = BreakRotator(InRot=BreakTransform(InTransform=GetDeltaTransformFromRefPose(self=@n10, BoneName=spine_05, BaseName=spine_01)).Rotation)
   Set backpack_rotate(execute=@n1.then_5, backpack_rotate=MakeRotator(Roll=MapRangeClamped(Value=@n11.Yaw, InRangeA=0, InRangeB=-20, OutRangeB=15)))
   > Set backpack_translation(backpack_translation=MakeVector(Y=MapRangeClamped(Value=@n11.Yaw, InRangeA=0, InRangeB=-30, OutRangeB=-5)))
   n12 = GetRealTimeSeconds()
   n13 = "Sin (Radians)"(A=float * float(A=float * float(A=1, B=-1.4) {is=title K2Node_PromotableOperator}, B=@n12) {is=title K2Node_CommutativeAssociativeBinaryOperator})
   n14 = "Sin (Radians)"(A=float * float(A=float * float(A=0.4, B=-0.8) {is=title K2Node_PromotableOperator}, B=@n12) {is=title K2Node_CommutativeAssociativeBinaryOperator})
   n15 = "Sin (Radians)"(A=float * float(A=float * float(A=1, B=2) {is=title K2Node_PromotableOperator}, B=@n12) {is=title K2Node_CommutativeAssociativeBinaryOperator})
   "Set Translate Front Gem "(
    execute=@n1.then_1
    "Translate Front Gem ":Vector=MakeVector(X=float / float(A=@n13, B=-1), Y=float / float(A=@n15, B=1), Z=float / float(A=@n14, B=-0.75))
   ) -> (Output_Get:Vector)
   n16 = "Sin (Radians)"(A=float * float(A=float * float(A=-1.3, B=-0.9) {is=title K2Node_PromotableOperator}, B=@n3) {is=title K2Node_CommutativeAssociativeBinaryOperator})
   n17 = "Sin (Radians)"(A=float * float(A=float * float(A=-1.2, B=-0.8) {is=title K2Node_PromotableOperator}, B=@n3) {is=title K2Node_CommutativeAssociativeBinaryOperator})
   n18 = "Sin (Radians)"(A=float * float(A=float * float(A=0.2, B=0.9) {is=title K2Node_PromotableOperator}, B=@n3) {is=title K2Node_CommutativeAssociativeBinaryOperator})
   Set Translate Back Gem 2(
    execute=@n1.then_3
    Translate Back Gem 2:Vector=MakeVector(X=float / float(A=@n16, B=-0.95), Y=float / float(A=@n18, B=-0.6), Z=float / float(A=@n17, B=-0.6))
   ) -> (Output_Get:Vector)
  ]}
]
variables [
 backpack_rotate {type=Rotator}
 backpack_translation {type=Vector}
 Gems RB Alpha {type=number}
 Gravity Override {type=Vector, default=<0 0 -980>}
 Gravity Override Params {type=GravityOverrideParamsStruct}
 NewVar {type=boolean}
 Translate Back Gem 1 {type=Vector}
 Translate Back Gem 2 {type=Vector}
 Translate Back Gem 3 {type=Vector}
 "Translate Front Gem " {type=Vector}
]
