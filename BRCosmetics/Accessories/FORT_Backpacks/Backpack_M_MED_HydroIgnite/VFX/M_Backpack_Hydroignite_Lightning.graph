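# Translucent unlit material for the lightning VFX of the Hydroignite backpack cosmetic, usable on Niagara ribbons and mesh particles.
# A lightning texture scrolls along U with a per-particle random offset.
# Dynamic parameters erode and boost its R channel, fade the G-channel glow and control an appearance gradient along U.
# The squared mask blends glow color into an HDR core color, which feeds an emissive adjustment function.
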Material /BRCosmetics/Accessories/FORT_Backpacks/Backpack_M_MED_HydroIgnite/VFX/M_Backpack_Hydroignite_Lightning
graph {
 properties {
  bUsedWithNiagaraMeshParticles=true
  bUsedWithNiagaraRibbons=true
  blendMode=BLEND_Translucent
  pixelDepthOffsetMode=PDOM_Legacy
  shadingModel=MSM_Unlit
  usedShadingModels=Unlit
 }
 nodes [
  n1 = DynamicParameter() {defaultValue=<rgba 1 0 0.4 1>, paramNames [LightningAppear, Dissolve, GlowFade, EmissiveBoost]}
  n2 = TextureCoordinate()
  n3 = AppendVector(A=ScalarParameter() {defaultValue=0.3, parameterName=UScale}, B=ScalarParameter() {defaultValue=1, parameterName=VScale})
  n4 = Add(A=ParticleRandom(), B=Multiply(A=Time(), B=ScalarParameter() {defaultValue=0.6, parameterName=UScrollSpeed}))
  n5 = Constant()
  n6 = TextureSample(UVs=Add(A=Multiply(A=@n2, B=@n3), B=AppendVector(A=@n4, B=@n5))) {
   samplerType=SAMPLERTYPE_LinearColor
   texture=&/T_Backpack_Hydroignite_Lightning
  }
  n7 = SmoothStep(Min=@n1.Dissolve, Value=@n6.R)
  n8 = OneMinus(None=ComponentMask(None=TextureCoordinate()) {r=true})
  n9 = Multiply(A=Max(A=Multiply(A=@n7, B=@n1.EmissiveBoost), B=Multiply(A=@n6.G, B=@n1.GlowFade)), B=Saturate(None=Add(A=@n1.LightningAppear, B=@n8)))
  MaterialFunctionCall(Emissive=LinearInterpolate(
    A=VectorParameter() {parameterName=ColorGlow, defaultValue=<rgba 0.032 0.8333 0.5268 1>}.RGB
    B=VectorParameter() {parameterName=ColorCore, defaultValue=<rgba 0.0637 7 5.5969 1>}.RGB
    Alpha=Multiply(A=@n9, B=@n9)
   )) {materialFunction=&MF_FX_EmissiveAdjustments}
 ]
}
