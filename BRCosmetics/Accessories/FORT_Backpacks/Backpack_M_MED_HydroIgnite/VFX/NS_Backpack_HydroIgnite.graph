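# Niagara system for the HydroIgnite backpack cosmetic, adding looping electric arcs, a glow and sparks around the backpack mesh.
# Paired lightning emitters for four sockets on User.SkeletalMesh bend burst ribbons along randomized Hermite splines to a shared point, with curve-animated material parameters.
# A center arc joins two fixed points, and some emitters randomly pick cyan or orange colors.
# Glow pulses a sprite and light; CenterLightning001 is disabled.
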NiagaraSystem /BRCosmetics/Accessories/FORT_Backpacks/Backpack_M_MED_HydroIgnite/VFX/NS_Backpack_HydroIgnite
dependencies {
 usedDataInterfaces [&NiagaraDataInterfaceCurve, &NiagaraDataInterfaceArrayColor, &NiagaraDataInterfaceArrayFloat3]
 usedDynamicInputs []
 usedModules [
  &SystemState
  &self:SystemSpawnScript.SystemScriptSource.SystemScriptGraph.NiagaraNodeAssignment_0.SetVariables_#1
 ]
}
emitters {
 Bit1_Lightining1 {
  emitterUpdateScript [
   &EmitterState(Inactive Response=::NewEnumerator1, Life Cycle Mode=::NewEnumerator1, Loop Duration=0.4, UseLoopDelay=false)
   /SetVariables(Emitter.ColorSeed:NiagaraInt32=&RandomRangeInt(Minimum=0, Maximum=1)) {isSetParameters=true, module=SetVariables_#2}
   /SetVariables(
    Emitter.CoreColor:LinearColor=&SelectLinearColorFromArray(
     Array Sampling Mode=::NewEnumerator1
     Direct Array Index=$Emitter.ColorSeed
     Color Selection Array=DataInterface() {propertyValues {ColorData [<rgba 0 10 8.9808 1>, <rgba 10 0.3625 0 1>]}}
     Random Seed=?
    )
    Emitter.GlowColor:LinearColor=&SelectLinearColorFromArray(
     Array Sampling Mode=::NewEnumerator1
     Direct Array Index=$Emitter.ColorSeed
     Color Selection Array=DataInterface() {propertyValues {ColorData [<rgba 0 0.5593 0.75 1>, <rgba 0.8333 0.2125 0 1>]}}
     Random Seed=?
    )
   ) {isSetParameters=true, module=SetVariables_#3}
   &SpawnBurst_Instantaneous(Loop Count Limit=3, Spawn Count=$System.RibbonSpawnCount, Use Spawn Probability=true, UseLoopCountLimit=false)
  ]
  particleSpawnScript [
   &InitializeParticle(Lifetime=0.4, Ribbon Width Mode=::NewEnumerator1, Sprite Size Mode=::NewEnumerator3, UsePositionOffset=false, Write Material Random=true)
   &PartitionParticles(Number Of Partitions=3)
  ]
  particleUpdateScript [
   &ParticleState()
   &SkeletalMeshLocation(
    Bone / Socket Index=4
    MaskSpawnGroup=false
    Mesh Sampling Type=::NewEnumerator5
    Skeletal Mesh=$User.SkeletalMesh
    Socket Sampling Mode=::NewEnumerator5
    Use Sampled Mesh Attribute as Ribbon ID=false
   )
   /Game/Effects/Niagara/Modules/Curves/FN_HermiteSpline(
    Direction End:Vector3f=&RandomRangeVector(Minimum=<6 0 -12>, Maximum=<-5 -10 35>)
    Position Blend:NiagaraFloat=&ReturnNormalizedExecIndex()
    Position End:Vector3f=<6 -8.4378 9>
    Position Start:Vector3f=$Particles.SkeletalMeshLocation.SampledPosition
    Write Particles Position=true
   )
   /SetVariables(Particles.RibbonLinkOrder:NiagaraFloat=&ReturnNormalizedExecIndex()) {isSetParameters=true, module=SetVariables_#4}
   &DynamicMaterialParameters(
    Index 0 Param 1=&FloatFromCurve(FloatCurve=DataInterface() {propertyValues {Curve {keys [
         {arriveTangent=-0.9313, interpMode=RCIM_Cubic, leaveTangent=-0.9313, tangentMode=RCTM_User, time=0, value=-1}
         {arriveTangent=7.0333, interpMode=RCIM_Cubic, leaveTangent=7.0333, tangentMode=RCTM_User, time=0.6, value=1}
        ]}}})
    Index 0 Param 2=&FloatFromCurve(FloatCurve=DataInterface() {propertyValues {Curve {keys [
         <at 0 0>
         {interpMode=RCIM_Cubic, leaveTangent=4.9411, tangentMode=RCTM_Break, time=0.45, value=0.05}
         {arriveTangent=0.7061, interpMode=RCIM_Cubic, leaveTangent=0.5589, tangentMode=RCTM_Break, time=0.6, value=0.7}
         {arriveTangent=1.5597, interpMode=RCIM_Cubic, tangentMode=RCTM_Break, time=0.9, value=1}
        ]}}})
    Index 0 Param 3=&FloatFromCurve(FloatCurve=DataInterface() {propertyValues {Curve {keys [
         <at 0 0>
         <at 0.2 0.1>
         {interpMode=RCIM_Cubic, tangentMode=RCTM_Break, time=0.4, value=0.3}
         {arriveTangent=-0.7389, interpMode=RCIM_Cubic, tangentMode=RCTM_Break, time=1, value=0}
        ]}}})
    Index 0 Param 4=&FloatFromCurve(FloatCurve=DataInterface() {propertyValues {Curve {keys [
         {interpMode=RCIM_Cubic, tangentMode=RCTM_Break, time=0.4, value=1}
         {arriveTangent=103.2496, interpMode=RCIM_Cubic, leaveTangent=-93.4054, tangentMode=RCTM_Break, time=0.6, value=10}
         <at 0.7 1>
        ]}}})
    Module.Index0Param0Write=true
    Module.Index0Param1Write=true
    Module.Index0Param2Write=true
    Module.Index0Param3Write=true
   )
  ]
  rendererClasses [&NiagaraSpriteRendererProperties, &NiagaraRibbonRendererProperties]
  renderers [
   {rendererClass=&NiagaraSpriteRendererProperties}
   {rendererClass=&NiagaraRibbonRendererProperties, rendererIndex=1}
  ]
 }
 Bit1_Lightining2 {
  emitterUpdateScript [
   &EmitterState(Inactive Response=::NewEnumerator1, Life Cycle Mode=::NewEnumerator1, Loop Duration=0.6, UseLoopDelay=false)
   &SpawnBurst_Instantaneous(Loop Count Limit=3, Spawn Count=$System.RibbonSpawnCount, Use Spawn Probability=true, UseLoopCountLimit=false)
  ]
  particleSpawnScript [
   &InitializeParticle(
    Color=$System.LightningColor
    Lifetime=0.6
    Ribbon Width Mode=::NewEnumerator1
    Sprite Size Mode=::NewEnumerator3
    UsePositionOffset=false
    Write Material Random=true
   )
   &PartitionParticles(Number Of Partitions=3)
  ]
  particleUpdateScript [
   &ParticleState()
   &SkeletalMeshLocation(
    Bone / Socket Index=4
    MaskSpawnGroup=false
    Mesh Sampling Type=::NewEnumerator5
    Skeletal Mesh=$User.SkeletalMesh
    Socket Sampling Mode=::NewEnumerator5
    Use Sampled Mesh Attribute as Ribbon ID=false
   )
   &FN_HermiteSpline(
    Direction End:Vector3f=&RandomRangeVector(Minimum=<10 0 -12>, Maximum=<-10 -16 35>)
    Position Blend:NiagaraFloat=&ReturnNormalizedExecIndex()
    Position End:Vector3f=<6 -8.4378 9>
    Position Start:Vector3f=$Particles.SkeletalMeshLocation.SampledPosition
    Write Particles Position=true
   )
   /SetVariables(Particles.RibbonLinkOrder:NiagaraFloat=&ReturnNormalizedExecIndex()) {isSetParameters=true, module=SetVariables_#4}
   &DynamicMaterialParameters(
    Index 0 Param 1=&FloatFromCurve(FloatCurve=DataInterface() {propertyValues {Curve {keys [
         {arriveTangent=-0.9313, interpMode=RCIM_Cubic, leaveTangent=-0.9313, tangentMode=RCTM_User, time=0, value=-1}
         {arriveTangent=7.0333, interpMode=RCIM_Cubic, leaveTangent=7.0333, tangentMode=RCTM_User, time=0.6, value=1}
        ]}}})
    Index 0 Param 2=&FloatFromCurve(FloatCurve=DataInterface() {propertyValues {Curve {keys [
         <at 0 0>
         {interpMode=RCIM_Cubic, leaveTangent=4.9411, tangentMode=RCTM_Break, time=0.45, value=0.05}
         {arriveTangent=0.7061, interpMode=RCIM_Cubic, leaveTangent=0.5589, tangentMode=RCTM_Break, time=0.6, value=0.7}
         {arriveTangent=1.5597, interpMode=RCIM_Cubic, tangentMode=RCTM_Break, time=0.9, value=1}
        ]}}})
    Index 0 Param 3=&FloatFromCurve(FloatCurve=DataInterface() {propertyValues {Curve {keys [
         <at 0 0>
         <at 0.2 0.1>
         {interpMode=RCIM_Cubic, tangentMode=RCTM_Break, time=0.4, value=0.3}
         {arriveTangent=-0.7389, interpMode=RCIM_Cubic, tangentMode=RCTM_Break, time=1, value=0}
        ]}}})
    Index 0 Param 4=&FloatFromCurve(FloatCurve=DataInterface() {propertyValues {Curve {keys [
         {interpMode=RCIM_Cubic, tangentMode=RCTM_Break, time=0.4, value=1}
         {arriveTangent=103.2496, interpMode=RCIM_Cubic, leaveTangent=-93.4054, tangentMode=RCTM_Break, time=0.6, value=10}
         <at 0.7 1>
        ]}}})
    Module.Index0Param0Write=true
    Module.Index0Param1Write=true
    Module.Index0Param2Write=true
    Module.Index0Param3Write=true
   )
  ]
  rendererClasses [&NiagaraSpriteRendererProperties, &NiagaraRibbonRendererProperties]
  renderers [
   {rendererClass=&NiagaraSpriteRendererProperties}
   {rendererClass=&NiagaraRibbonRendererProperties, rendererIndex=1}
  ]
 }
 Bit2_Lightining1 {
  emitterUpdateScript [
   &EmitterState(Inactive Response=::NewEnumerator1, Life Cycle Mode=::NewEnumerator1, Loop Duration=0.5, UseLoopDelay=false)
   /SetVariables(Emitter.ColorSeed:NiagaraInt32=&RandomRangeInt(Minimum=0, Maximum=1)) {isSetParameters=true, module=SetVariables_#2}
   /SetVariables(
    Emitter.CoreColor:LinearColor=&SelectLinearColorFromArray(
     Array Sampling Mode=::NewEnumerator1
     Direct Array Index=$Emitter.ColorSeed
     Color Selection Array=DataInterface() {propertyValues {ColorData [<rgba 0 10 8.9808 1>, <rgba 10 0.3625 0 1>]}}
     Random Seed=?
    )
    Emitter.GlowColor:LinearColor=&SelectLinearColorFromArray(
     Array Sampling Mode=::NewEnumerator1
     Direct Array Index=$Emitter.ColorSeed
     Color Selection Array=DataInterface() {propertyValues {ColorData [<rgba 0 0.5593 0.75 1>, <rgba 0.8333 0.2125 0 1>]}}
     Random Seed=?
    )
   ) {isSetParameters=true, module=SetVariables_#3}
   &SpawnBurst_Instantaneous(Loop Count Limit=3, Spawn Count=$System.RibbonSpawnCount, Spawn Probability=0.8, Use Spawn Probability=true, UseLoopCountLimit=false)
  ]
  particleSpawnScript [
   &InitializeParticle(
    Color=$System.LightningColor
    Lifetime=0.5
    Ribbon Width Mode=::NewEnumerator1
    Sprite Size Mode=::NewEnumerator3
    UsePositionOffset=false
    Write Material Random=true
   )
   &PartitionParticles(Number Of Partitions=3)
  ]
  particleUpdateScript [
   &ParticleState()
   &SkeletalMeshLocation(
    Bone / Socket Index=3
    MaskSpawnGroup=false
    Mesh Sampling Type=::NewEnumerator5
    Skeletal Mesh=$User.SkeletalMesh
    Socket Sampling Mode=::NewEnumerator5
    Use Sampled Mesh Attribute as Ribbon ID=false
   )
   &FN_HermiteSpline(
    Direction End:Vector3f=&RandomRangeVector(Minimum=<6 0 -12>, Maximum=<-5 -10 35>)
    Position Blend:NiagaraFloat=&ReturnNormalizedExecIndex()
    Position End:Vector3f=<6 -8.4378 9>
    Position Start:Vector3f=$Particles.SkeletalMeshLocation.SampledPosition
    Write Particles Position=true
   )
   /SetVariables(Particles.RibbonLinkOrder:NiagaraFloat=&ReturnNormalizedExecIndex()) {isSetParameters=true, module=SetVariables_#4}
   &DynamicMaterialParameters(
    Index 0 Param 1=&FloatFromCurve(FloatCurve=DataInterface() {propertyValues {Curve {keys [
         {arriveTangent=-0.9313, interpMode=RCIM_Cubic, leaveTangent=-0.9313, tangentMode=RCTM_User, time=0, value=-1}
         {arriveTangent=7.0333, interpMode=RCIM_Cubic, leaveTangent=7.0333, tangentMode=RCTM_User, time=0.6, value=1}
        ]}}})
    Index 0 Param 2=&FloatFromCurve(FloatCurve=DataInterface() {propertyValues {Curve {keys [
         <at 0 0>
         {interpMode=RCIM_Cubic, leaveTangent=4.9411, tangentMode=RCTM_Break, time=0.45, value=0.05}
         {arriveTangent=0.7061, interpMode=RCIM_Cubic, leaveTangent=0.5589, tangentMode=RCTM_Break, time=0.6, value=0.7}
         {arriveTangent=1.5597, interpMode=RCIM_Cubic, tangentMode=RCTM_Break, time=0.9, value=1}
        ]}}})
    Index 0 Param 3=&FloatFromCurve(FloatCurve=DataInterface() {propertyValues {Curve {keys [
         <at 0 0>
         <at 0.2 0.1>
         {interpMode=RCIM_Cubic, tangentMode=RCTM_Break, time=0.4, value=0.3}
         {arriveTangent=-0.7389, interpMode=RCIM_Cubic, tangentMode=RCTM_Break, time=1, value=0}
        ]}}})
    Index 0 Param 4=&FloatFromCurve(FloatCurve=DataInterface() {propertyValues {Curve {keys [
         {interpMode=RCIM_Cubic, tangentMode=RCTM_Break, time=0.4, value=1}
         {arriveTangent=103.2496, interpMode=RCIM_Cubic, leaveTangent=-93.4054, tangentMode=RCTM_Break, time=0.6, value=10}
         <at 0.7 1>
        ]}}})
    Module.Index0Param0Write=true
    Module.Index0Param1Write=true
    Module.Index0Param2Write=true
    Module.Index0Param3Write=true
   )
  ]
  rendererClasses [&NiagaraSpriteRendererProperties, &NiagaraRibbonRendererProperties]
  renderers [
   {rendererClass=&NiagaraSpriteRendererProperties}
   {rendererClass=&NiagaraRibbonRendererProperties, rendererIndex=1}
  ]
 }
 Bit2_Lightining2 {
  emitterUpdateScript [
   &EmitterState(Inactive Response=::NewEnumerator1, Life Cycle Mode=::NewEnumerator1, Loop Duration=0.6, UseLoopDelay=false)
   &SpawnBurst_Instantaneous(Loop Count Limit=3, Spawn Count=$System.RibbonSpawnCount, Use Spawn Probability=true, UseLoopCountLimit=false)
  ]
  particleSpawnScript [
   &InitializeParticle(
    Color=$System.LightningColor
    Lifetime=0.6
    Ribbon Width Mode=::NewEnumerator1
    Sprite Size Mode=::NewEnumerator3
    UsePositionOffset=false
    Write Material Random=true
   )
   &PartitionParticles(Number Of Partitions=3)
  ]
  particleUpdateScript [
   &ParticleState()
   &SkeletalMeshLocation(
    Bone / Socket Index=3
    MaskSpawnGroup=false
    Mesh Sampling Type=::NewEnumerator5
    Skeletal Mesh=$User.SkeletalMesh
    Socket Sampling Mode=::NewEnumerator5
    Use Sampled Mesh Attribute as Ribbon ID=false
   )
   &FN_HermiteSpline(
    Direction End:Vector3f=&RandomRangeVector(Minimum=<10 0 -20>, Maximum=<-10 -16 55>)
    Position Blend:NiagaraFloat=&ReturnNormalizedExecIndex()
    Position End:Vector3f=<6 -8.4378 9>
    Position Start:Vector3f=$Particles.SkeletalMeshLocation.SampledPosition
    Write Particles Position=true
   )
   /SetVariables(Particles.RibbonLinkOrder:NiagaraFloat=&ReturnNormalizedExecIndex()) {isSetParameters=true, module=SetVariables_#4}
   &DynamicMaterialParameters(
    Index 0 Param 1=&FloatFromCurve(FloatCurve=DataInterface() {propertyValues {Curve {keys [
         {arriveTangent=-0.9313, interpMode=RCIM_Cubic, leaveTangent=-0.9313, tangentMode=RCTM_User, time=0, value=-1}
         {arriveTangent=7.0333, interpMode=RCIM_Cubic, leaveTangent=7.0333, tangentMode=RCTM_User, time=0.6, value=1}
        ]}}})
    Index 0 Param 2=&FloatFromCurve(FloatCurve=DataInterface() {propertyValues {Curve {keys [
         <at 0 0>
         {interpMode=RCIM_Cubic, leaveTangent=4.9411, tangentMode=RCTM_Break, time=0.45, value=0.05}
         {arriveTangent=0.7061, interpMode=RCIM_Cubic, leaveTangent=0.5589, tangentMode=RCTM_Break, time=0.6, value=0.7}
         {arriveTangent=1.5597, interpMode=RCIM_Cubic, tangentMode=RCTM_Break, time=0.9, value=1}
        ]}}})
    Index 0 Param 3=&FloatFromCurve(FloatCurve=DataInterface() {propertyValues {Curve {keys [
         <at 0 0>
         <at 0.2 0.1>
         {interpMode=RCIM_Cubic, tangentMode=RCTM_Break, time=0.4, value=0.3}
         {arriveTangent=-0.7389, interpMode=RCIM_Cubic, tangentMode=RCTM_Break, time=1, value=0}
        ]}}})
    Index 0 Param 4=&FloatFromCurve(FloatCurve=DataInterface() {propertyValues {Curve {keys [
         {interpMode=RCIM_Cubic, tangentMode=RCTM_Break, time=0.4, value=1}
         {arriveTangent=103.2496, interpMode=RCIM_Cubic, leaveTangent=-93.4054, tangentMode=RCTM_Break, time=0.6, value=10}
         <at 0.7 1>
        ]}}})
    Module.Index0Param0Write=true
    Module.Index0Param1Write=true
    Module.Index0Param2Write=true
    Module.Index0Param3Write=true
   )
  ]
  rendererClasses [&NiagaraSpriteRendererProperties, &NiagaraRibbonRendererProperties]
  renderers [
   {rendererClass=&NiagaraSpriteRendererProperties}
   {rendererClass=&NiagaraRibbonRendererProperties, rendererIndex=1}
  ]
 }
 Bit3_Lightining1 {
  emitterUpdateScript [
   &EmitterState(Inactive Response=::NewEnumerator1, Life Cycle Mode=::NewEnumerator1, Loop Duration=0.6, UseLoopDelay=false)
   &SpawnBurst_Instantaneous(Loop Count Limit=3, Spawn Count=$System.RibbonSpawnCount, Spawn Probability=0.7, Use Spawn Probability=true, UseLoopCountLimit=false)
   /SetVariables(Emitter.ColorSeed:NiagaraInt32=&RandomRangeInt(Minimum=0, Maximum=1)) {isSetParameters=true, module=SetVariables_#2}
   /SetVariables(
    Emitter.CoreColor:LinearColor=&SelectLinearColorFromArray(
     Array Sampling Mode=::NewEnumerator1
     Direct Array Index=$Emitter.ColorSeed
     Color Selection Array=DataInterface() {propertyValues {ColorData [<rgba 0 10 8.9808 1>, <rgba 10 0.3625 0 1>]}}
     Random Seed=?
    )
    Emitter.GlowColor:LinearColor=&SelectLinearColorFromArray(
     Array Sampling Mode=::NewEnumerator1
     Direct Array Index=$Emitter.ColorSeed
     Color Selection Array=DataInterface() {propertyValues {ColorData [<rgba 0 0.5593 0.75 1>, <rgba 0.8333 0.2125 0 1>]}}
     Random Seed=?
    )
   ) {isSetParameters=true, module=SetVariables_#3}
  ]
  particleSpawnScript [
   &InitializeParticle(
    Color=$System.LightningColor
    Lifetime=0.6
    Ribbon Width Mode=::NewEnumerator1
    Sprite Size Mode=::NewEnumerator3
    UsePositionOffset=false
    Write Material Random=true
   )
   &PartitionParticles(Number Of Partitions=3)
  ]
  particleUpdateScript [
   &ParticleState()
   &SkeletalMeshLocation(
    Bone / Socket Index=1
    MaskSpawnGroup=false
    Mesh Sampling Type=::NewEnumerator5
    Skeletal Mesh=$User.SkeletalMesh
    Socket Sampling Mode=::NewEnumerator5
    Use Sampled Mesh Attribute as Ribbon ID=false
   )
   &FN_HermiteSpline(
    Direction End:Vector3f=&RandomRangeVector(Minimum=<6 0 -12>, Maximum=<-5 -10 35>)
    Position Blend:NiagaraFloat=&ReturnNormalizedExecIndex()
    Position End:Vector3f=<6 -8.4378 9>
    Position Start:Vector3f=$Particles.SkeletalMeshLocation.SampledPosition
    Write Particles Position=true
   )
   /SetVariables(Particles.RibbonLinkOrder:NiagaraFloat=&ReturnNormalizedExecIndex()) {isSetParameters=true, module=SetVariables_#4}
   &DynamicMaterialParameters(
    Index 0 Param 1=&FloatFromCurve(FloatCurve=DataInterface() {propertyValues {Curve {keys [
         {arriveTangent=-0.9313, interpMode=RCIM_Cubic, leaveTangent=-0.9313, tangentMode=RCTM_User, time=0, value=-1}
         {arriveTangent=7.0333, interpMode=RCIM_Cubic, leaveTangent=7.0333, tangentMode=RCTM_User, time=0.6, value=1}
        ]}}})
    Index 0 Param 2=&FloatFromCurve(FloatCurve=DataInterface() {propertyValues {Curve {keys [
         <at 0 0>
         {interpMode=RCIM_Cubic, leaveTangent=4.9411, tangentMode=RCTM_Break, time=0.45, value=0.05}
         {arriveTangent=0.7061, interpMode=RCIM_Cubic, leaveTangent=0.5589, tangentMode=RCTM_Break, time=0.6, value=0.7}
         {arriveTangent=1.5597, interpMode=RCIM_Cubic, tangentMode=RCTM_Break, time=0.9, value=1}
        ]}}})
    Index 0 Param 3=&FloatFromCurve(FloatCurve=DataInterface() {propertyValues {Curve {keys [
         <at 0 0>
         <at 0.2 0.1>
         {interpMode=RCIM_Cubic, tangentMode=RCTM_Break, time=0.4, value=0.3}
         {arriveTangent=-0.7389, interpMode=RCIM_Cubic, tangentMode=RCTM_Break, time=1, value=0}
        ]}}})
    Index 0 Param 4=&FloatFromCurve(FloatCurve=DataInterface() {propertyValues {Curve {keys [
         {interpMode=RCIM_Cubic, tangentMode=RCTM_Break, time=0.4, value=1}
         {arriveTangent=103.2496, interpMode=RCIM_Cubic, leaveTangent=-93.4054, tangentMode=RCTM_Break, time=0.6, value=10}
         <at 0.7 1>
        ]}}})
    Module.Index0Param0Write=true
    Module.Index0Param1Write=true
    Module.Index0Param2Write=true
    Module.Index0Param3Write=true
   )
  ]
  rendererClasses [&NiagaraSpriteRendererProperties, &NiagaraRibbonRendererProperties]
  renderers [
   {rendererClass=&NiagaraSpriteRendererProperties}
   {rendererClass=&NiagaraRibbonRendererProperties, rendererIndex=1}
  ]
 }
 Bit3_Lightining2 {
  emitterUpdateScript [
   &EmitterState(Inactive Response=::NewEnumerator1, Life Cycle Mode=::NewEnumerator1, Loop Duration=0.6, UseLoopDelay=false)
   &SpawnBurst_Instantaneous(Loop Count Limit=3, Spawn Count=$System.RibbonSpawnCount, Use Spawn Probability=true, UseLoopCountLimit=false)
  ]
  particleSpawnScript [
   &InitializeParticle(
    Color=$System.LightningColor
    Lifetime=0.6
    Ribbon Width Mode=::NewEnumerator1
    Sprite Size Mode=::NewEnumerator3
    UsePositionOffset=false
    Write Material Random=true
   )
   &PartitionParticles(Number Of Partitions=3)
  ]
  particleUpdateScript [
   &ParticleState()
   &SkeletalMeshLocation(
    Bone / Socket Index=1
    MaskSpawnGroup=false
    Mesh Sampling Type=::NewEnumerator5
    Skeletal Mesh=$User.SkeletalMesh
    Socket Sampling Mode=::NewEnumerator5
    Use Sampled Mesh Attribute as Ribbon ID=false
   )
   &FN_HermiteSpline(
    Direction End:Vector3f=&RandomRangeVector(Minimum=<10 0 -12>, Maximum=<-10 -16 30>)
    Position Blend:NiagaraFloat=&ReturnNormalizedExecIndex()
    Position End:Vector3f=<6 -8.4378 9>
    Position Start:Vector3f=$Particles.SkeletalMeshLocation.SampledPosition
    Write Particles Position=true
   )
   /SetVariables(Particles.RibbonLinkOrder:NiagaraFloat=&ReturnNormalizedExecIndex()) {isSetParameters=true, module=SetVariables_#4}
   &DynamicMaterialParameters(
    Index 0 Param 1=&FloatFromCurve(FloatCurve=DataInterface() {propertyValues {Curve {keys [
         {arriveTangent=-0.9313, interpMode=RCIM_Cubic, leaveTangent=-0.9313, tangentMode=RCTM_User, time=0, value=-1}
         {arriveTangent=7.0333, interpMode=RCIM_Cubic, leaveTangent=7.0333, tangentMode=RCTM_User, time=0.6, value=1}
        ]}}})
    Index 0 Param 2=&FloatFromCurve(FloatCurve=DataInterface() {propertyValues {Curve {keys [
         <at 0 0>
         {interpMode=RCIM_Cubic, leaveTangent=4.9411, tangentMode=RCTM_Break, time=0.45, value=0.05}
         {arriveTangent=0.7061, interpMode=RCIM_Cubic, leaveTangent=0.5589, tangentMode=RCTM_Break, time=0.6, value=0.7}
         {arriveTangent=1.5597, interpMode=RCIM_Cubic, tangentMode=RCTM_Break, time=0.9, value=1}
        ]}}})
    Index 0 Param 3=&FloatFromCurve(FloatCurve=DataInterface() {propertyValues {Curve {keys [
         <at 0 0>
         <at 0.2 0.1>
         {interpMode=RCIM_Cubic, tangentMode=RCTM_Break, time=0.4, value=0.3}
         {arriveTangent=-0.7389, interpMode=RCIM_Cubic, tangentMode=RCTM_Break, time=1, value=0}
        ]}}})
    Index 0 Param 4=&FloatFromCurve(FloatCurve=DataInterface() {propertyValues {Curve {keys [
         {interpMode=RCIM_Cubic, tangentMode=RCTM_Break, time=0.4, value=1}
         {arriveTangent=103.2496, interpMode=RCIM_Cubic, leaveTangent=-93.4054, tangentMode=RCTM_Break, time=0.6, value=10}
         <at 0.7 1>
        ]}}})
    Module.Index0Param0Write=true
    Module.Index0Param1Write=true
    Module.Index0Param2Write=true
    Module.Index0Param3Write=true
   )
  ]
  rendererClasses [&NiagaraSpriteRendererProperties, &NiagaraRibbonRendererProperties]
  renderers [
   {rendererClass=&NiagaraSpriteRendererProperties}
   {rendererClass=&NiagaraRibbonRendererProperties, rendererIndex=1}
  ]
 }
 Bit4_Lightining1 {
  emitterUpdateScript [
   &EmitterState(Inactive Response=::NewEnumerator1, Life Cycle Mode=::NewEnumerator1, Loop Duration=0.35, UseLoopDelay=false)
   &SpawnBurst_Instantaneous(Loop Count Limit=3, Spawn Count=$System.RibbonSpawnCount, Spawn Probability=0.5, Use Spawn Probability=true, UseLoopCountLimit=false)
   /SetVariables(Emitter.ColorSeed:NiagaraInt32=&RandomRangeInt(Minimum=0, Maximum=1)) {isSetParameters=true, module=SetVariables_#2}
   /SetVariables(
    Emitter.CoreColor:LinearColor=&SelectLinearColorFromArray(
     Array Sampling Mode=::NewEnumerator1
     Direct Array Index=$Emitter.ColorSeed
     Color Selection Array=DataInterface() {propertyValues {ColorData [<rgba 0 10 8.9808 1>, <rgba 10 0.3625 0 1>]}}
     Random Seed=?
    )
    Emitter.GlowColor:LinearColor=&SelectLinearColorFromArray(
     Array Sampling Mode=::NewEnumerator1
     Direct Array Index=$Emitter.ColorSeed
     Color Selection Array=DataInterface() {propertyValues {ColorData [<rgba 0 0.5593 0.75 1>, <rgba 0.8333 0.2125 0 1>]}}
     Random Seed=?
    )
   ) {isSetParameters=true, module=SetVariables_#3}
  ]
  particleSpawnScript [
   &InitializeParticle(
    Color=$System.LightningColor
    Lifetime=0.35
    Ribbon Width Mode=::NewEnumerator1
    Sprite Size Mode=::NewEnumerator3
    UsePositionOffset=false
    Write Material Random=true
   )
   &PartitionParticles(Number Of Partitions=3)
  ]
  particleUpdateScript [
   &ParticleState()
   &SkeletalMeshLocation(
    Bone / Socket Index=2
    MaskSpawnGroup=false
    Mesh Sampling Type=::NewEnumerator5
    Skeletal Mesh=$User.SkeletalMesh
    Socket Sampling Mode=::NewEnumerator5
    Use Sampled Mesh Attribute as Ribbon ID=false
   )
   &FN_HermiteSpline(
    Direction End:Vector3f=&RandomRangeVector(Minimum=<6 0 -12>, Maximum=<-5 -10 35>)
    Position Blend:NiagaraFloat=&ReturnNormalizedExecIndex()
    Position End:Vector3f=<6 -8.4378 9>
    Position Start:Vector3f=$Particles.SkeletalMeshLocation.SampledPosition
    Write Particles Position=true
   )
   /SetVariables(Particles.RibbonLinkOrder:NiagaraFloat=&ReturnNormalizedExecIndex()) {isSetParameters=true, module=SetVariables_#4}
   &DynamicMaterialParameters(
    Index 0 Param 1=&FloatFromCurve(FloatCurve=DataInterface() {propertyValues {Curve {keys [
         {arriveTangent=-0.9313, interpMode=RCIM_Cubic, leaveTangent=-0.9313, tangentMode=RCTM_User, time=0, value=-1}
         {arriveTangent=7.0333, interpMode=RCIM_Cubic, leaveTangent=7.0333, tangentMode=RCTM_User, time=0.6, value=1}
        ]}}})
    Index 0 Param 2=&FloatFromCurve(FloatCurve=DataInterface() {propertyValues {Curve {keys [
         <at 0 0>
         {interpMode=RCIM_Cubic, leaveTangent=4.9411, tangentMode=RCTM_Break, time=0.45, value=0.05}
         {arriveTangent=0.7061, interpMode=RCIM_Cubic, leaveTangent=0.5589, tangentMode=RCTM_Break, time=0.6, value=0.7}
         {arriveTangent=1.5597, interpMode=RCIM_Cubic, tangentMode=RCTM_Break, time=0.9, value=1}
        ]}}})
    Index 0 Param 3=&FloatFromCurve(FloatCurve=DataInterface() {propertyValues {Curve {keys [
         <at 0 0>
         <at 0.2 0.1>
         {interpMode=RCIM_Cubic, tangentMode=RCTM_Break, time=0.4, value=0.3}
         {arriveTangent=-0.7389, interpMode=RCIM_Cubic, tangentMode=RCTM_Break, time=1, value=0}
        ]}}})
    Index 0 Param 4=&FloatFromCurve(FloatCurve=DataInterface() {propertyValues {Curve {keys [
         {interpMode=RCIM_Cubic, tangentMode=RCTM_Break, time=0.4, value=1}
         {arriveTangent=103.2496, interpMode=RCIM_Cubic, leaveTangent=-93.4054, tangentMode=RCTM_Break, time=0.6, value=10}
         <at 0.7 1>
        ]}}})
    Module.Index0Param0Write=true
    Module.Index0Param1Write=true
    Module.Index0Param2Write=true
    Module.Index0Param3Write=true
   )
  ]
  rendererClasses [&NiagaraSpriteRendererProperties, &NiagaraRibbonRendererProperties]
  renderers [
   {rendererClass=&NiagaraSpriteRendererProperties}
   {rendererClass=&NiagaraRibbonRendererProperties, rendererIndex=1}
  ]
 }
 Bit4_Lightining2 {
  emitterUpdateScript [
   &EmitterState(Inactive Response=::NewEnumerator1, Life Cycle Mode=::NewEnumerator1, Loop Duration=0.6, UseLoopDelay=false)
   &SpawnBurst_Instantaneous(Loop Count Limit=3, Spawn Count=$System.RibbonSpawnCount, Use Spawn Probability=true, UseLoopCountLimit=false)
  ]
  particleSpawnScript [
   &InitializeParticle(
    Color=$System.LightningColor
    Lifetime=0.6
    Ribbon Width Mode=::NewEnumerator1
    Sprite Size Mode=::NewEnumerator3
    UsePositionOffset=false
    Write Material Random=true
   )
   &PartitionParticles(Number Of Partitions=3)
  ]
  particleUpdateScript [
   &ParticleState()
   &SkeletalMeshLocation(
    Bone / Socket Index=2
    MaskSpawnGroup=false
    Mesh Sampling Type=::NewEnumerator5
    Skeletal Mesh=$User.SkeletalMesh
    Socket Sampling Mode=::NewEnumerator5
    Use Sampled Mesh Attribute as Ribbon ID=false
   )
   &FN_HermiteSpline(
    Direction End:Vector3f=&RandomRangeVector(Minimum=<10 0 0>, Maximum=<30 0 0>)
    Position Blend:NiagaraFloat=&ReturnNormalizedExecIndex()
    Position End:Vector3f=<6 -8.4378 9>
    Position Start:Vector3f=$Particles.SkeletalMeshLocation.SampledPosition
    Write Particles Position=true
   )
   /SetVariables(Particles.RibbonLinkOrder:NiagaraFloat=&ReturnNormalizedExecIndex()) {isSetParameters=true, module=SetVariables_#4}
   &DynamicMaterialParameters(
    Index 0 Param 1=&FloatFromCurve(FloatCurve=DataInterface() {propertyValues {Curve {keys [
         {arriveTangent=-0.9313, interpMode=RCIM_Cubic, leaveTangent=-0.9313, tangentMode=RCTM_User, time=0, value=-1}
         {arriveTangent=7.0333, interpMode=RCIM_Cubic, leaveTangent=7.0333, tangentMode=RCTM_User, time=0.6, value=1}
        ]}}})
    Index 0 Param 2=&FloatFromCurve(FloatCurve=DataInterface() {propertyValues {Curve {keys [
         <at 0 0>
         {interpMode=RCIM_Cubic, leaveTangent=4.9411, tangentMode=RCTM_Break, time=0.45, value=0.05}
         {arriveTangent=0.7061, interpMode=RCIM_Cubic, leaveTangent=0.5589, tangentMode=RCTM_Break, time=0.6, value=0.7}
         {arriveTangent=1.5597, interpMode=RCIM_Cubic, tangentMode=RCTM_Break, time=0.9, value=1}
        ]}}})
    Index 0 Param 3=&FloatFromCurve(FloatCurve=DataInterface() {propertyValues {Curve {keys [
         <at 0 0>
         <at 0.2 0.1>
         {interpMode=RCIM_Cubic, tangentMode=RCTM_Break, time=0.4, value=0.3}
         {arriveTangent=-0.7389, interpMode=RCIM_Cubic, tangentMode=RCTM_Break, time=1, value=0}
        ]}}})
    Index 0 Param 4=&FloatFromCurve(FloatCurve=DataInterface() {propertyValues {Curve {keys [
         {interpMode=RCIM_Cubic, tangentMode=RCTM_Break, time=0.4, value=1}
         {arriveTangent=103.2496, interpMode=RCIM_Cubic, leaveTangent=-93.4054, tangentMode=RCTM_Break, time=0.6, value=10}
         <at 0.7 1>
        ]}}})
    Module.Index0Param0Write=true
    Module.Index0Param1Write=true
    Module.Index0Param2Write=true
    Module.Index0Param3Write=true
   )
  ]
  rendererClasses [&NiagaraSpriteRendererProperties, &NiagaraRibbonRendererProperties]
  renderers [
   {rendererClass=&NiagaraSpriteRendererProperties}
   {rendererClass=&NiagaraRibbonRendererProperties, rendererIndex=1}
  ]
 }
 CenterLightning {
  emitterUpdateScript [
   &EmitterState(Inactive Response=::NewEnumerator1, Life Cycle Mode=::NewEnumerator1, Loop Duration=0.4, UseLoopDelay=false)
   /SetVariables(Emitter.ColorSeed:NiagaraInt32=&RandomRangeInt(Minimum=0, Maximum=1)) {isSetParameters=true, module=SetVariables_#2}
   /SetVariables(
    Emitter.CoreColor:LinearColor=&SelectLinearColorFromArray(
     Array Sampling Mode=::NewEnumerator1
     Direct Array Index=$Emitter.ColorSeed
     Color Selection Array=DataInterface() {propertyValues {ColorData [<rgba 0 10 8.9808 1>, <rgba 10 0.3625 0 1>]}}
     Random Seed=?
    )
    Emitter.GlowColor:LinearColor=&SelectLinearColorFromArray(
     Array Sampling Mode=::NewEnumerator1
     Direct Array Index=$Emitter.ColorSeed
     Color Selection Array=DataInterface() {propertyValues {ColorData [<rgba 0 0.5593 0.75 1>, <rgba 0.8333 0.2125 0 1>]}}
     Random Seed=?
    )
   ) {isSetParameters=true, module=SetVariables_#3}
   &SpawnBurst_Instantaneous(Loop Count Limit=3, Spawn Count=12, Use Spawn Probability=true, UseLoopCountLimit=false)
  ]
  particleSpawnScript [
   &InitializeParticle(
    Lifetime=0.4
    Ribbon Width=6
    Ribbon Width Mode=::NewEnumerator1
    Sprite Size Mode=::NewEnumerator3
    UsePositionOffset=false
    Write Material Random=true
   )
   &PartitionParticles(Number Of Partitions=3)
  ]
  particleUpdateScript [
   &ParticleState()
   &SkeletalMeshLocation(
    Bone / Socket Index=4
    MaskSpawnGroup=false
    Mesh Sampling Type=::NewEnumerator5
    Skeletal Mesh=$User.SkeletalMesh
    Socket Sampling Mode=::NewEnumerator5
    Use Sampled Mesh Attribute as Ribbon ID=false
   )
   &FN_HermiteSpline(
    Direction End:Vector3f=&SelectVectorFromArray(
     Array Sampling Mode=::NewEnumerator0
     Direct Array Index=?
     Vector Selection Array=DataInterface() {propertyValues {FloatData [<25 0 -40>, <-25 0 40>]}}
     Random Seed=0
    )
    Position Blend:NiagaraFloat=&ReturnNormalizedExecIndex()
    Position End:Vector3f=<8 -15 8>
    Position Start:Vector3f=<8 -7.3477 8>
    Write Particles Position=true
   )
   /SetVariables(Particles.RibbonLinkOrder:NiagaraFloat=&ReturnNormalizedExecIndex()) {isSetParameters=true, module=SetVariables_#4}
   &DynamicMaterialParameters(
    Index 0 Param 1=&FloatFromCurve(FloatCurve=DataInterface() {propertyValues {Curve {keys [
         {arriveTangent=-0.9313, interpMode=RCIM_Cubic, leaveTangent=-0.9313, tangentMode=RCTM_User, time=0, value=-1}
         {arriveTangent=7.0333, interpMode=RCIM_Cubic, leaveTangent=7.0333, tangentMode=RCTM_User, time=0.6, value=1}
        ]}}})
    Index 0 Param 2=&FloatFromCurve(FloatCurve=DataInterface() {propertyValues {Curve {keys [
         <at 0 0>
         {interpMode=RCIM_Cubic, leaveTangent=4.9411, tangentMode=RCTM_Break, time=0.45, value=0.05}
         {arriveTangent=0.7061, interpMode=RCIM_Cubic, leaveTangent=0.5589, tangentMode=RCTM_Break, time=0.6, value=0.7}
         {arriveTangent=1.5597, interpMode=RCIM_Cubic, tangentMode=RCTM_Break, time=0.9, value=1}
        ]}}})
    Index 0 Param 3=&FloatFromCurve(FloatCurve=DataInterface() {propertyValues {Curve {keys [
         <at 0 0>
         <at 0.2 0.1>
         {interpMode=RCIM_Cubic, tangentMode=RCTM_Break, time=0.4, value=0.3}
         {arriveTangent=-0.7389, interpMode=RCIM_Cubic, tangentMode=RCTM_Break, time=1, value=0}
        ]}}})
    Index 0 Param 4=&FloatFromCurve(FloatCurve=DataInterface() {propertyValues {Curve {keys [
         {interpMode=RCIM_Cubic, tangentMode=RCTM_Break, time=0.4, value=1}
         {arriveTangent=103.2496, interpMode=RCIM_Cubic, leaveTangent=-93.4054, tangentMode=RCTM_Break, time=0.6, value=10}
         <at 0.7 1>
        ]}}})
    Module.Index0Param0Write=true
    Module.Index0Param1Write=true
    Module.Index0Param2Write=true
    Module.Index0Param3Write=true
   )
  ]
  rendererClasses [&NiagaraSpriteRendererProperties, &NiagaraRibbonRendererProperties]
  renderers [
   {rendererClass=&NiagaraSpriteRendererProperties}
   {rendererClass=&NiagaraRibbonRendererProperties, rendererIndex=1}
  ]
 }
 CenterLightning001 {
  bEnabled=false
  emitterUpdateScript [
   &EmitterState(Life Cycle Mode=::NewEnumerator1, Loop Duration=0.4, UseLoopDelay=false)
   /SetVariables(Emitter.ColorSeed:NiagaraInt32=&RandomRangeInt(Minimum=0, Maximum=1)) {isSetParameters=true, module=SetVariables_#2}
   /SetVariables(
    Emitter.CoreColor:LinearColor=&SelectLinearColorFromArray(
     Array Sampling Mode=::NewEnumerator1
     Direct Array Index=$Emitter.ColorSeed
     Color Selection Array=DataInterface() {propertyValues {ColorData [<rgba 0 10 8.9808 1>, <rgba 10 0.3625 0 1>]}}
     Random Seed=?
    )
    Emitter.GlowColor:LinearColor=&SelectLinearColorFromArray(
     Array Sampling Mode=::NewEnumerator1
     Direct Array Index=$Emitter.ColorSeed
     Color Selection Array=DataInterface() {propertyValues {ColorData [<rgba 0 0.5593 0.75 1>, <rgba 0.8333 0.2125 0 1>]}}
     Random Seed=?
    )
   ) {isSetParameters=true, module=SetVariables_#3}
   &SpawnBurst_Instantaneous(Loop Count Limit=3, Spawn Count=5, Use Spawn Probability=true, UseLoopCountLimit=false)
  ]
  particleSpawnScript [
   &InitializeParticle(
    Lifetime=0.4
    Ribbon Width=8
    Ribbon Width Mode=::NewEnumerator1
    Sprite Size Mode=::NewEnumerator3
    UsePositionOffset=false
    Write Material Random=true
   )
   &PartitionParticles(Number Of Partitions=3)
  ]
  particleUpdateScript [
   &ParticleState()
   &SkeletalMeshLocation(
    Bone / Socket Index=4
    MaskSpawnGroup=false
    Mesh Sampling Type=::NewEnumerator5
    Skeletal Mesh=$User.SkeletalMesh
    Socket Sampling Mode=::NewEnumerator5
    Use Sampled Mesh Attribute as Ribbon ID=false
   )
   &FN_HermiteSpline(
    Direction End:Vector3f=&SelectVectorFromArray(
     Array Sampling Mode=::NewEnumerator0
     Direct Array Index=?
     Vector Selection Array=DataInterface() {propertyValues {FloatData [<25 3 -40>, <-25 3 40>]}}
     Random Seed=0
    )
    Position Blend:NiagaraFloat=&ReturnNormalizedExecIndex()
    Position End:Vector3f=<8 -15 8>
    Position Start:Vector3f=<8 -7.3477 8>
    Write Particles Position=true
   )
   /SetVariables(Particles.RibbonLinkOrder:NiagaraFloat=&ReturnNormalizedExecIndex()) {isSetParameters=true, module=SetVariables_#4}
   &DynamicMaterialParameters(
    Index 0 Param 1=&FloatFromCurve(FloatCurve=DataInterface() {propertyValues {Curve {keys [
         {arriveTangent=-0.9313, interpMode=RCIM_Cubic, leaveTangent=-0.9313, tangentMode=RCTM_User, time=0, value=-1}
         {arriveTangent=7.0333, interpMode=RCIM_Cubic, leaveTangent=7.0333, tangentMode=RCTM_User, time=0.6, value=1}
        ]}}})
    Index 0 Param 2=&FloatFromCurve(FloatCurve=DataInterface() {propertyValues {Curve {keys [
         <at 0 0>
         {interpMode=RCIM_Cubic, leaveTangent=4.9411, tangentMode=RCTM_Break, time=0.45, value=0.05}
         {arriveTangent=0.7061, interpMode=RCIM_Cubic, leaveTangent=0.5589, tangentMode=RCTM_Break, time=0.6, value=0.7}
         {arriveTangent=1.5597, interpMode=RCIM_Cubic, tangentMode=RCTM_Break, time=0.9, value=1}
        ]}}})
    Index 0 Param 3=&FloatFromCurve(FloatCurve=DataInterface() {propertyValues {Curve {keys [
         <at 0 0>
         <at 0.2 0.1>
         {interpMode=RCIM_Cubic, tangentMode=RCTM_Break, time=0.4, value=0.3}
         {arriveTangent=-0.7389, interpMode=RCIM_Cubic, tangentMode=RCTM_Break, time=1, value=0}
        ]}}})
    Index 0 Param 4=&FloatFromCurve(FloatCurve=DataInterface() {propertyValues {Curve {keys [
         {interpMode=RCIM_Cubic, tangentMode=RCTM_Break, time=0.4, value=1}
         {arriveTangent=103.2496, interpMode=RCIM_Cubic, leaveTangent=-93.4054, tangentMode=RCTM_Break, time=0.6, value=10}
         <at 0.7 1>
        ]}}})
    Module.Index0Param0Write=true
    Module.Index0Param1Write=true
    Module.Index0Param2Write=true
    Module.Index0Param3Write=true
   )
  ]
  rendererClasses [&NiagaraSpriteRendererProperties, &NiagaraRibbonRendererProperties]
  renderers [
   {rendererClass=&NiagaraSpriteRendererProperties}
   {rendererClass=&NiagaraRibbonRendererProperties, rendererIndex=1}
  ]
 }
 Glow {
  emitterUpdateScript [
   &EmitterState(
    Inactive Response=::NewEnumerator1
    Life Cycle Mode=::NewEnumerator1
    Loop Behavior=::NewEnumerator1
    Loop Duration Mode=::NewEnumerator1
    UseLoopDelay=false
   )
   &SpawnBurst_Instantaneous(Use Spawn Probability=false, UseLoopCountLimit=false)
  ]
  particleSpawnScript [
   &FN_SetSpawnGroup(Number of Spawn Groups=1)
   &InitializeParticle(
    Color=<rgba 0 0.8078 1 0.35>
    Lifetime=5
    Sprite Size Mode=::NewEnumerator3
    Uniform Sprite Size=50
    UsePositionOffset=false
    Write Material Random=true
   )
   &SkeletalMeshLocation(
    MaskSpawnGroup=false
    Mesh Sampling Type=::NewEnumerator5
    Sampled Position Offset Coordinate Space:ENiagaraSkelMeshTransforms=::NewEnumerator0
    Skeletal Mesh=$User.SkeletalMesh
    Socket Sampling Mode=::NewEnumerator5
   )
   &CameraOffset(Camera Offset Amount=8)
  ]
  particleUpdateScript [
   &ParticleState(Kill Particles When Lifetime Has Elapsed=false)
   &ScaleColor(
    Scale Alpha=&FloatFromCurve(
     FloatCurve=DataInterface() {propertyValues {Curve {keys [
         {interpMode=RCIM_Cubic, leaveTangent=0.5345, tangentMode=RCTM_Break, time=0, value=0.95}
         {interpMode=RCIM_Cubic, tangentMode=RCTM_Break, time=0.5, value=0.5}
         {arriveTangent=-0.0692, interpMode=RCIM_Cubic, tangentMode=RCTM_Break, time=1, value=0.95}
        ]}}}
     CurveIndex=$System.LoopedAge
    )
    ScaleA=true
    ScaleRGB=true
   )
   &FN_Light_v2(Light Exponent=2, Light Radius=40, LightEnabled=true, Scale Color=0.5)
  ]
  rendererClasses [&NiagaraSpriteRendererProperties, &NiagaraLightRendererProperties]
  renderers [
   {rendererClass=&NiagaraSpriteRendererProperties}
   {rendererClass=&NiagaraLightRendererProperties, rendererIndex=1}
  ]
 }
 Sparks {
  emitterUpdateScript [
   &EmitterState()
   &SpawnRate(Spawn Probability=0.5, SpawnRate=18, Use Spawn Probability=true)
  ]
  particleSpawnScript [
   &InitializeParticle(
    Color=<rgba 0 10 9.6998 1>
    Lifetime Max=0.5
    Lifetime Min=0.25
    Lifetime Mode=::NewEnumerator1
    Sprite Size Mode=::NewEnumerator1
    Uniform Sprite Size Max=1.5
    Uniform Sprite Size Min=1
    UsePositionOffset=false
    Write Material Random=true
   )
   &SkeletalMeshLocation(MaskSpawnGroup=false, Mesh Sampling Type=::NewEnumerator5, Skeletal Mesh=$User.SkeletalMesh, Socket Sampling Mode=::NewEnumerator5)
   &AddVelocity(
    Module.Use Falloff Distance=false
    Override Seed=false
    Velocity Mode=::NewEnumerator1
    Velocity Origin=$Particles.SkeletalMeshLocation.SampledPosition
    Velocity Speed=&RandomRangeFloat(Minimum=100, Maximum=200)
   )
  ]
  particleUpdateScript [
   &ParticleState()
   &ScaleSpriteSize(Uniform Curve Sprite Scale=DataInterface() {propertyValues {Curve {keys [
        {arriveTangent=9.0725, interpMode=RCIM_Cubic, leaveTangent=9.0725, tangentMode=RCTM_User, time=0, value=0}
        {arriveTangent=-2.5249, interpMode=RCIM_Cubic, leaveTangent=0.4866, tangentMode=RCTM_Break, time=0.1408, value=0.9847}
        {arriveTangent=-2.1048, interpMode=RCIM_Cubic, leaveTangent=-2.1048, tangentMode=RCTM_User, time=1, value=0}
       ]}}})
   &Drag(Drag=10, Use Linear Drag=true, Use Rotational Drag=true)
   &GravityForce(Gravity=<0 0 -20>)
   &SolveForcesAndVelocity(Clamp Velocity=false, Limit Acceleration=false)
  ]
  rendererClasses [&NiagaraSpriteRendererProperties]
  renderers [{rendererClass=&NiagaraSpriteRendererProperties}]
 }
}
summary {
 systemName=NS_Backpack_HydroIgnite
 emitters [
  {emitterName=Glow, rendererClasses [&NiagaraSpriteRendererProperties, &NiagaraLightRendererProperties]}
  {emitterName=Sparks, rendererClasses [&NiagaraSpriteRendererProperties]}
  {
   emitterName=Bit1_Lightining1
   rendererClasses [&NiagaraSpriteRendererProperties, &NiagaraRibbonRendererProperties]
  }
  {
   emitterName=Bit1_Lightining2
   rendererClasses [&NiagaraSpriteRendererProperties, &NiagaraRibbonRendererProperties]
  }
  {
   emitterName=Bit2_Lightining2
   rendererClasses [&NiagaraSpriteRendererProperties, &NiagaraRibbonRendererProperties]
  }
  {
   emitterName=Bit3_Lightining2
   rendererClasses [&NiagaraSpriteRendererProperties, &NiagaraRibbonRendererProperties]
  }
  {
   emitterName=Bit4_Lightining2
   rendererClasses [&NiagaraSpriteRendererProperties, &NiagaraRibbonRendererProperties]
  }
  {
   emitterName=Bit2_Lightining1
   rendererClasses [&NiagaraSpriteRendererProperties, &NiagaraRibbonRendererProperties]
  }
  {
   emitterName=Bit3_Lightining1
   rendererClasses [&NiagaraSpriteRendererProperties, &NiagaraRibbonRendererProperties]
  }
  {
   emitterName=Bit4_Lightining1
   rendererClasses [&NiagaraSpriteRendererProperties, &NiagaraRibbonRendererProperties]
  }
  {
   emitterName=CenterLightning
   rendererClasses [&NiagaraSpriteRendererProperties, &NiagaraRibbonRendererProperties]
  }
  {
   bEnabled=false
   emitterName=CenterLightning001
   rendererClasses [&NiagaraSpriteRendererProperties, &NiagaraRibbonRendererProperties]
  }
 ]
 userVariables [User.SkeletalMesh {
   type=NiagaraDataInterfaceSkeletalMesh
   defaultValue {
    struct=NiagaraExt_VariableValue_DataInterface
    value {dataInterface=&self:NiagaraDataInterfaceSkeletalMesh_0}
   }
  }]
}
